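# Material function for the Hightower Tapas player skin, layering space and electric emissive effects onto incoming material attributes.
# A camera-facing virtual plane near the actor, scaled by FOV, projects rotating star textures into FX-mask red areas.
# Step-scrolled, normal-distorted noise and GmapMask glow drive time-of-day-scaled emissive, tinted by Tapas-Blue.
# Static switches select head, face accessory, hammer and holo-foil variants, changing masks and effect layers.
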MaterialFunction /Game/Characters/Player/Male/Medium/Bodies/M_MED_HighTower_Tapas/Materials/MF_Hightower_Tapas
graph {nodes [
  Constant3Vector() {constant=<rgba 0.24 4.9291 12 1>}
  Constant3Vector() {constant=<rgba 0.01 0.1112 0.5 1>}
  Constant() {r=70}
  Constant() {r=5}
  Constant() {r=6}
  FunctionInput() {bUsePreviewValueAsDefault=true, inputName=SkinMask, inputType=FunctionInput_Scalar}
  n1 = Constant() {desc=Needs at least a small amount of this or it will break, r=0.25}
  n2 = MaterialFunctionCall() {materialFunction=&CameraDirectionVector}
  n3 = Normalize(VectorInput=Multiply(A=Constant3Vector() {constant=<rgba 1 1 0 0>}, B=@n2.Result))
  n4 = Constant() {r=25}
  n5 = Add(A=ActorPositionWS(), B=Constant3Vector() {constant=<rgba 0 0 40 0>})
  n6 = MaterialFunctionCall(
   Input=MaterialFunctionCall() {materialFunction=/Engine/Functions/Engine_MaterialFunctions02/Utility/FOV}."FOV "
   Input Low=Constant() {r=0.414}
   Input High=Constant() {r=1.091}
   Target Low=Constant()
   Target High=Constant() {r=1}
  ) {materialFunction=&RemapValueRange}
  n7 = MaterialFunctionCall(
   Plane Normal Axis=Add(A=Multiply(A=@n1, B=@n3), B=@n2.Result)
   Plane Center=Add(A=Multiply(A=@n2.Result, B=@n4), B=@n5)
   UV Scale=LinearInterpolate(A=Constant() {r=75}, B=Constant() {r=200}, Alpha=Saturate(None=@n6.Result))
  ) {materialFunction=&VirtualPlaneCoordinates}
  n8 = TextureSample(UVs=Multiply(A=@n7.UVs, B=Constant() {r=2})) {
   samplerSource=SSM_Wrap_WorldGroupSettings
   texture=/Game/Characters/Player/Male/Medium/Bodies/M_MED_HighTower_Tapas/Textures/T_Hightower_Tapas_Star
  }
  n9 = Constant3Vector() {constant=<rgba 1.8 3.4747 6 1>}
  n10 = Constant() {r=3}
  n11 = Time()
  n12 = Constant() {r=0.0005}
  n13 = TextureSample(UVs=MaterialFunctionCall(UVs=Multiply(A=@n7.UVs, B=@n10), "Rotation Angle (0-1)"=Multiply(A=@n11, B=@n12)) {materialFunction=&CustomRotator}.Rotated Values) {
   samplerSource=SSM_Wrap_WorldGroupSettings
   texture=&T_Hightower_Tapas_Star
  }
  n14 = Constant() {r=0.5}
  n15 = Constant() {r=-0.0025}
  n16 = TextureSample(UVs=MaterialFunctionCall(UVs=@n7.UVs, "Rotation Angle (0-1)"=Multiply(A=@n11, B=@n15)) {materialFunction=&CustomRotator}.Rotated Values) {
   samplerSource=SSM_Wrap_WorldGroupSettings
   texture=&T_Hightower_Tapas_Star
  }
  n17 = Multiply(A=Add(A=Multiply(A=@n13.R, B=@n14), B=@n16.R), B=@n9)
  n18 = SphereMask(A=Constant2Vector() {g=0.5, r=0.5}, B=@n7.UVs, Radius=Constant() {r=0.4}, Hardness=Constant()) {attenuationRadius=0.5}
  n19 = Constant() {r=5}
  n20 = ScalarParameter() {group=Tapas, parameterName=Tapas-Blue}
  n21 = Multiply(A=Multiply(A=Power(Base=@n18, Exp=@n19), B=Constant3Vector() {constant=<rgba 0.045 0.1467 0.3 1>}), B=@n20)
  n22 = Power(
   Base=OneMinus(None=FunctionInput() {bUsePreviewValueAsDefault=true, inputName=Fresnel, inputType=FunctionInput_Scalar})
   Exp=Constant() {r=3}
  )
  n23 = Multiply(A=@n22, B=StaticSwitchParameter(True=Multiply(A=@n8.R, B=@n9), False=Add(A=@n17, B=@n21)) {group=Tapas, parameterName=Tapas-Use-HoloFoil})
  n24 = Constant()
  n25 = FunctionInput() {inputType=FunctionInput_MaterialAttributes}
  n26 = TextureSampleParameter2D() {
   group=Tapas
   parameterName="FXmask Red=Space Green=Emissive"
   samplerSource=SSM_Wrap_WorldGroupSettings
   texture=&T_BlacksRGB
  }
  n27 = StaticSwitchParameter(
   True=BlendMaterialAttributes(
    A=@n25
    B=SetMaterialAttributes!(MaterialAttributes=@n25, EmissiveColor=@n23, BaseColor=@n24) {attributeSetTypes [#1, #2]}
    Alpha=@n26.R
   ) {vertexAttributeBlendType=UseA}
   False=@n25
  ) {group=Tapas, parameterName=Tapas-UseFX-Mask}
  n28 = GetMaterialAttributes(None=@n27) {desc=Eye emissive texture, attributeGetTypes [#1]}
  n29 = Constant() {r=1.1}
  n30 = Constant() {r=1}
  n31 = Constant() {r=0.65}
  n32 = ScalarParameter() {defaultValue=0.3, group=Tapas, parameterName=Tapas-EmissiveNightMod}
  n33 = Constant3Vector() {constant=<rgba 0.0025 0.2441 0.5 1>}
  n34 = Constant3Vector() {constant=<rgba 1 1.3987 2 1>}
  n35 = FunctionInput() {bUsePreviewValueAsDefault=true, inputName=GmapMask, inputType=FunctionInput_Scalar, previewValue=<1 1 1 1>}
  n36 = LinearInterpolate(A=Constant(), B=Constant() {r=12}, Alpha=Power(Base=@n35, Exp=Constant() {r=3}))
  n37 = Constant() {r=8}
  n38 = Saturate(None=LinearInterpolate(A=Constant(), B=Constant() {r=4}, Alpha=Power(Base=@n35, Exp=@n37)))
  n39 = Constant() {r=6}
  n40 = Saturate(None=LinearInterpolate(A=Constant(), B=Constant() {r=4}, Alpha=Power(Base=@n35, Exp=@n39)))
  n41 = Multiply(
   A=StaticSwitchParameter(True=Saturate(None=@n36), False=StaticSwitchParameter(True=@n38, False=@n40) {group=Tapas, parameterName=Tapas-IsFaceAccessory}) {group=Tapas, parameterName=Tapas-IsHead}
   B=@n22
  )
  n42 = Constant3Vector() {constant=<rgba 0.007 0.1233 0.35 1>}
  n43 = Constant3Vector() {constant=<rgba 1 2.5949 5 1>}
  n44 = LinearInterpolate(A=LinearInterpolate(A=@n33, B=@n34, Alpha=@n41), B=LinearInterpolate(A=@n42, B=@n43, Alpha=@n41), Alpha=@n20) {desc="A=Normal\r\nB=Blue"}
  n45 = Multiply(A=ScalarParameter() {defaultValue=5, desc=Altered in the pickaxe mat, group=Tapas, parameterName=Tapas-ElectricTileA}, B=@n7.UVs)
  n46 = Time()
  n47 = Constant() {r=0.75}
  n48 = Frac(None=Multiply(A=Multiply(A=@n46, B=@n47), B=Constant2Vector() {g=0.5, r=1}))
  n49 = Constant() {desc=Can be 2-vector, r=12}
  n50 = Divide(A=Ceil(None=Multiply(A=@n48, B=@n49) {constB=30}), B=Constant() {desc=Can be 2-vector, r=0.9})
  n51 = GetMaterialAttributes(None=@n25) {attributeGetTypes [#3, #4]}
  n52 = Multiply(A=Add(A=ComponentMask(None=@n51.Normal) {g=true, r=true}, B=Constant() {r=-0.5}), B=Constant() {r=-0.05})
  n53 = TextureSample(UVs=Add(A=Add(A=@n45, B=@n50), B=@n52)) {
   samplerSource=SSM_Wrap_WorldGroupSettings
   samplerType=SAMPLERTYPE_LinearColor
   texture=/Game/Characters/Player/Male/Medium/Bodies/M_MED_SRDrift/Textures/T_ElectricNoise
  }
  n54 = Multiply(A=@n7.UVs, B=ScalarParameter() {defaultValue=-2, desc=Altered in the pickaxe mat, group=Tapas, parameterName=Tapas-ElectricTileB})
  n55 = TextureSample(UVs=Add(A=Add(A=@n50, B=@n54), B=@n52)) {
   samplerSource=SSM_Wrap_WorldGroupSettings
   samplerType=SAMPLERTYPE_LinearColor
   texture=&T_ElectricNoise
  }
  n56 = Constant() {r=10}
  n57 = Constant() {r=-4}
  n58 = Multiply(A=Saturate(None=LinearInterpolate(A=@n56, B=@n57, Alpha=@n51.Metallic)), B=@n20)
  n59 = ScalarParameter() {group=Tapas, parameterName=Tapas-IsHeadMat}
  n60 = Subtract(
   A=Power(Base=Max(A=@n53.G, B=@n55.G), Exp=Constant() {r=2.5})
   B=Multiply(A=Constant() {r=0.5}, B=StaticSwitchParameter(True=@n58, False=@n59) {group=Tapas, parameterName=Tapas-IsFaceAccessory})
  )
  n61 = Multiply(A=Constant() {r=0.5}, B=@n60)
  n62 = Multiply(A=@n60, B=Constant() {r=2})
  n63 = LinearInterpolate(A=LinearInterpolate(A=@n33, B=@n34, Alpha=Max(A=@n41, B=@n61)), B=LinearInterpolate(A=@n42, B=@n43, Alpha=Max(A=@n41, B=@n62)), Alpha=@n20) {desc="A=Normal\r\nB=Blue"}
  n64 = Add(
   A=Multiply(
    A=MaterialFunctionCall(Value_Morning=@n29, Value_Day=@n30, Value_Evening=@n31, Value_Night=@n32) {materialFunction=&Scalar_Time_of_Day}.Result
    B=StaticSwitchParameter(True=@n44, False=@n63) {group=Tapas, parameterName=Tapas-Use-HoloFoil}
   )
   B=@n28.EmissiveColor
  )
  n65 = Constant()
  n66 = ShadingModel()
  n67 = Multiply(A=@n59, B=@n20)
  n68 = StaticSwitchParameter(True=@n26.G, False=Constant()) {group=Tapas, parameterName=Tapas-UseFX-Mask}
  n69 = StaticSwitchParameter(True=Saturate(None=Add(A=@n58, B=@n68)), False=StaticSwitchParameter(True=@n20, False=@n68) {group=Tapas, parameterName=Tapas-IsHammer}) {group=Tapas, parameterName=Tapas-IsFaceAccessory}
  FunctionOutput(None=BlendMaterialAttributes(
    A=@n27
    B=SetMaterialAttributes!(MaterialAttributes=@n28.MaterialAttributes, EmissiveColor=@n64, BaseColor=@n65, ShadingModel=@n66) {attributeSetTypes [#1, #2, #5]}
    Alpha=StaticSwitchParameter(True=@n67, False=@n69) {group=Tapas, parameterName=Tapas-IsHead}
   ) {vertexAttributeBlendType=UseA})
  n70 = Saturate(None=Add(B=Constant()))
  n71 = Constant() {r=10}
  n72 = Constant() {r=-4}
  n73 = Multiply(A=Saturate(None=LinearInterpolate(A=@n71, B=@n72)))
  n74 = Constant()
  LinearInterpolate(A=@n70, B=@n73, Alpha=@n74)
  Multiply(A=Constant3Vector() {constant=<rgba 0 0 1 0>}, B=ScalarParameter() {defaultValue=40, group=Tapas, parameterName=Tapas-AcreenAlignOffset})
  n75 = ScalarParameter() {defaultValue=6, desc="Used in hair && head mat", group=Tapas, parameterName=Tapas-GmapPower}
  Saturate(None=LinearInterpolate(
    A=Constant()
    B=ScalarParameter() {defaultValue=4, desc="Used in hair && head mat", group=Tapas, parameterName=Tapas-GmapLerpB}
    Alpha=Power(Base=@n35, Exp=@n75)
   ))
  Saturate(None=Add(A=Multiply(A=@n70), B=LinearInterpolate(B=@n73, Alpha=@n74)))
  StaticSwitchParameter(True=ObjectPositionWS(), False=@n5) {group=Tapas, parameterName=Tapas_UseObjectPos}
 ]}
