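# Master material for the male medium HighTower Tapas player body, using Default Lit and Subsurface Profile shading through material attributes.
# The shared character PreFX function feeds the Tapas-specific function, whose result goes into the shared PostFX function.
# A static switch optionally inserts the HoloFoil function borrowed from HighTower Squash before PostFX.
# HoloFoil receives vertex-interpolated pre-skinned position and normal, keeping its inputs independent of skeletal animation.
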Material /Game/Characters/Player/Male/Medium/Bodies/M_MED_HighTower_Tapas/Materials/M_MED_HighTower_Tapas_Master
graph {
 properties {
  bAutomaticallySetUsageInEditor=false
  bUseMaterialAttributes=true
  bUsedWithClothing=true
  bUsedWithMeshDeformer=true
  bUsedWithMorphTargets=true
  bUsedWithSkeletalMesh=true
  pixelDepthOffsetMode=PDOM_Legacy
  shadingModel=MSM_FromMaterialExpression
  subsurfaceProfile=&SS_HeroSkin_02
  usedShadingModels=Default Lit | Subsurface Profile
 }
 nodes [
  n1 = MaterialFunctionCall() {materialFunction=&BaseFN_CharacterFunction_PreFX}
  n2 = VertexInterpolator(VS=PreSkinnedPosition())
  n3 = VertexInterpolator(VS=PreSkinnedNormal())
  n4 = MaterialFunctionCall!(SkinMask=@n1.Skin, GmapMask=@n1.Skin, Fresnel=@n1.Skin, In=@n1.Skin) {materialFunction=&/MF_Hightower_Tapas}
  MaterialFunctionCall(
   AO=@n1.AO
   Distance Blend=@n1.AO
   Fresnel=@n1.AO
   Gmap=@n1.AO
   Material=StaticSwitchParameter(
    True=MaterialFunctionCall!(PreskinnedLocalPosition=@n2.PS, PreskinnedLocalNormal=@n3.PS, Material=@n4.Result) {materialFunction=/Game/Characters/Player/Female/Medium/Bodies/F_MED_HighTower_Squash/Materials/MF_HoloFoil}.Result
    False=@n4.Result
   ) {parameterName=BUseHoloFoil}
  ) {materialFunction=&BaseFN_CharacterFunction_PostFX}
 ]
}
